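# Sound cue for the thrust start of the Donut Plate glider cosmetic.
# A Random node picks among three thrust wave variants with equal weight.
# Playback is quietened by a low volume multiplier and uses the shared player glider thrust attenuation and gliders land sound class.
# It references a glider thrust concurrency set, alongside override settings that limit instances per owner and stop the oldest.
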SoundCue /BRCosmetics/Sounds/Gliders/Donut_Plate/Glider_ThrustStart_DonutPlate
root=Random_0
nodes [
 {id=Random_0, children [WavePlayer_1,WavePlayer_3,WavePlayer_4], properties {weights [1,1,1]}}
 {id=WavePlayer_1, properties {soundWaveAssetPtr=&/Glider_DonutPlate_Thrust_03}}
 {id=WavePlayer_3, properties {soundWaveAssetPtr=&/Glider_DonutPlate_Thrust_01}}
 {id=WavePlayer_4, properties {soundWaveAssetPtr=&/Glider_DonutPlate_Thrust_02}}
]
properties {
 attenuationSettings=&Fort_Player_Glider_Thrust_Athena_Att
 maxDistance=12500
 soundClassObject=&Fort_Gliders_Land
 volumeMultiplier=0.2
 concurrencyOverrides {bLimitToOwner=true, maxCount=2, resolutionRule=StopOldest}
 concurrencySet [&Athena_Glider_Thrust_Conc]
}
